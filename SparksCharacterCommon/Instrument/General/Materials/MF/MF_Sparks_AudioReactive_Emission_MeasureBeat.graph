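# Material function for Sparks instrument materials that scales the emissive color of incoming material attributes in time with music.
# The summed BeatLifetime and MeasureLifetime values from the Sparks audio parameter collection, clamped to 0-1, blend between min and max multipliers.
# A Use Flicker input blends that toward an MF_FX_Flicker-driven multiplier, offset by per-instance random by default.
# An On/Off input, on by default, blends between original and scaled emissive.
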MaterialFunction /SparksCharacterCommon/Instrument/General/Materials/MF/MF_Sparks_AudioReactive_Emission_MeasureBeat
graph {nodes [
  n1 = GetMaterialAttributes(None=FunctionInput() {description=In Material Attributes, inputName=MA, inputType=FunctionInput_MaterialAttributes, sortPriority=-1}) {attributeGetTypes [
    #1
    #2
    #3
    #4
   ]}
  n2 = Add(
   A=CollectionParameter() {collection=/FMAudioReactiveData/AudioAnalysis/MPC_Sparks_Audio_Parameters, parameterName=BeatLifetime}
   B=CollectionParameter() {collection=&MPC_Sparks_Audio_Parameters, parameterName=MeasureLifetime}
  )
  n3 = FunctionInput(Preview=Constant() {desc=Min emission scale, r=0.35}) {bUsePreviewValueAsDefault=true, inputName=Min Mult, inputType=FunctionInput_Scalar, sortPriority=1}
  n4 = FunctionInput(Preview=Constant() {desc=max emission scale, r=1.5}) {bUsePreviewValueAsDefault=true, inputName=Max Mult, inputType=FunctionInput_Scalar, sortPriority=2}
  n5 = FunctionInput(Preview=Constant() {desc=Flicker Speed Mult, r=0.5}) {bUsePreviewValueAsDefault=true, description=0-1 values only, inputName=Flicker Speed Mult, inputType=FunctionInput_Scalar, sortPriority=4}
  n6 = FunctionInput(Preview=PerInstanceRandom()) {bUsePreviewValueAsDefault=true, description=0-1 values only, inputName=Flicker Time Offset, inputType=FunctionInput_Scalar, sortPriority=5}
  n7 = MaterialFunctionCall!(Time Mult=@n5, Time Add=@n6) {materialFunction=/Game/Effects/Niagara/MasterMaterials/Functions/Mojo_MF/MF_FX_Flicker}
  n8 = FunctionInput(Preview=Constant() {desc=Use flicker?}) {bUsePreviewValueAsDefault=true, description=0-1 values only, inputName=Use Flicker?, inputType=FunctionInput_Scalar, sortPriority=3}
  n9 = LinearInterpolate(
   A=LinearInterpolate(A=@n3, B=@n4, Alpha=Saturate(None=@n2)) {constA=0.35, constB=2}
   B=LinearInterpolate(A=@n3, B=@n4, Alpha=Saturate(None=@n7.Raw Flicker))
   Alpha=Saturate(None=@n8)
  )
  n10 = FunctionInput(Preview=Constant() {r=1}) {
   bUsePreviewValueAsDefault=true
   desc="Audio Reactive Emission (Measure Beat)  0 or 1 value"
   inputName="Audio Reactive Emission (Measure Beat) On/Off"
   inputType=FunctionInput_Scalar
  }
  n11 = SetMaterialAttributes!(
   MaterialAttributes=@n1.MaterialAttributes
   EmissiveColor=LinearInterpolate(A=@n1.EmissiveColor, B=Multiply(A=@n1.EmissiveColor, B=@n9), Alpha=Saturate(None=@n10))
   Roughness
   BaseColor
   WorldPositionOffset
   Opacity
  ) {attributeSetTypes [
    #3
    #2
    #1
    #4
    #5
   ]}
  FunctionOutput(None=@n11)
 ]}
