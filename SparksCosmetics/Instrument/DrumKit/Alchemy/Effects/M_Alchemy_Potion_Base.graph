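# Base Default Lit material for the potion effect on the Sparks Alchemy drum kit, usable on skeletal meshes with mesh deformers.
# MF_Alchemy_Potion builds the material attributes, which a static switch can route through measure-beat audio-reactive emission.
# That emission exposes multiplier parameters and a flicker that fakes reactivity where no music plays, such as the front-end store.
# MF_FX_EmissiveAdjustments then processes only the emissive color before output.
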Material /SparksCosmetics/Instrument/DrumKit/Alchemy/Effects/M_Alchemy_Potion_Base
graph {
 properties {
  bUseMaterialAttributes=true
  bUsedWithMeshDeformer=true
  bUsedWithSkeletalMesh=true
  pixelDepthOffsetMode=PDOM_Legacy
  usedShadingModels=Default Lit
 }
 nodes [
  n1 = ScalarParameter() {
   defaultValue=1
   group="Audio Reactive Emission (Measure Beat)"
   parameterName="Audio Reactive Emission (Measure Beat) On/Off"
   sortPriority=-5
  }
  n2 = ScalarParameter() {defaultValue=3, group="Audio Reactive Emission (Measure Beat)", parameterName=Min Mult, sortPriority=-4}
  n3 = ScalarParameter() {defaultValue=1, group="Audio Reactive Emission (Measure Beat)", parameterName=Max Mult, sortPriority=-3}
  n4 = ScalarParameter() {
   desc="0-1 value, Only use if you only want ambient flicker and no audio reactivity (Ex: in the front end store where there is no music)"
   group="Audio Reactive Emission (Measure Beat)"
   parameterName="Use Flicker (Fake Audio Reactivity)"
   sortPriority=-1
  }
  n5 = ScalarParameter() {defaultValue=0.75, group="Audio Reactive Emission (Measure Beat)", parameterName=Flicker Speed Mult, sortPriority=-1}
  n6 = MaterialFunctionCall!(In=GetMaterialAttributes().MaterialAttributes) {desc=Sparks NEW, materialFunction=&/MF_Alchemy_Potion}
  n7 = GetMaterialAttributes(None=StaticSwitchParameter(
    True=MaterialFunctionCall!(
     MA=@n6.Result
     "Audio Reactive Emission (Measure Beat) On/Off"=@n1
     Min Mult=@n2
     Max Mult=@n3
     Use Flicker?=@n4
     Flicker Speed Mult=@n5
     Flicker Time Offset
    ) {materialFunction=/SparksCharacterCommon/Instrument/General/Materials/MF/MF_Sparks_AudioReactive_Emission_MeasureBeat}.Result
    False=@n6.Result
   ) {group="Audio Reactive Emission (Measure Beat)", parameterName="Enable Audio Reactive Emission (Measure Beat)"}) {attributeGetTypes [#1]}
  SetMaterialAttributes(
   MaterialAttributes=@n7.MaterialAttributes
   Emissive Color=MaterialFunctionCall(Emissive=@n7.Emissive Color) {materialFunction=&MF_FX_EmissiveAdjustments}.Result
  ) {attributeSetTypes [#1]}
 ]
}
