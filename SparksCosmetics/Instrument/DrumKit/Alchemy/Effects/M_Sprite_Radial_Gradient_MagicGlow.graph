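# Additive, unlit Niagara sprite material for a radial magic glow in the Sparks Alchemy drum kit effects.
# Particle color drives emissive, pulsed by the BeatLifetime value from the audio-reactive parameter collection.
# Opacity combines a bottom-masked sphere gradient, shaped by dynamic-parameter power and depth, angle and camera fades, with faint radial equalizer bars.
# Those bars fake FFT motion by feeding Perlin noise through MF_FX_Fake_FFT_Noise rather than reading spectrum data.
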Material /SparksCosmetics/Instrument/DrumKit/Alchemy/Effects/M_Sprite_Radial_Gradient_MagicGlow
graph {
 properties {
  bUsedWithNanite=true
  bUsedWithNiagaraSprites=true
  blendMode=BLEND_Additive
  pixelDepthOffsetMode=PDOM_Legacy
  shadingModel=MSM_Unlit
  twoSided=true
  usedShadingModels=Unlit
 }
 nodes [
  Constant() {r=1}
  Constant() {desc=test audio high end, r=1.5}
  n1 = ParticleColor()
  n2 = Sine(None=CollectionParameter() {collection=/FMAudioReactiveData/AudioAnalysis/MPC_Sparks_Audio_Parameters, parameterName=BeatLifetime})
  n3 = Saturate(None=Power(Base=ConstantBiasScale(None=@n2)) {constExponent=4})
  n4 = Constant()
  n5 = Constant() {r=1}
  n6 = Constant() {r=1}
  n7 = Constant() {r=1.5}
  MaterialFunctionCall(Emissive=Multiply(
    A=@n1.RGB
    B=MaterialFunctionCall(Input=@n3, Input Low=@n4, Input High=@n5, Target Low=@n6, Target High=@n7) {materialFunction=&RemapValueRange}.Result
   ) {desc=Audio Reactivity Emission Mult}) {materialFunction=&MF_FX_EmissiveAdjustments}
  n8 = ComponentMask(None=TextureCoordinate()) {g=true}
  NamedRerouteDeclaration(None=Step(Y=@n8) {constX=0.5}) {name=BottomCutMask, nodeColor=<rgba 0 0.5547 1 1>}
  NamedRerouteDeclaration(None=MaterialFunctionCall() {materialFunction=&VectorToRadialValue}.Radial Coordinates) {name=UV, nodeColor=<rgba 0.7578 0 1 1>}
  n9 = ComponentMask(None=NamedRerouteUsage()) {r=true}
  n10 = NamedRerouteDeclaration(None=Constant() {r=30}) {name=BarAmount, nodeColor=<rgba 1 0.5391 0 1>}
  n11 = Sine(None=Add(A=Multiply(A=@n9, B=@n10) {constB=10}) {constB=0.2})
  n12 = Constant() {r=1}
  NamedRerouteDeclaration(None=Step(Y=ConstantBiasScale(None=@n11), X=NamedRerouteDeclaration(None=@n12) {name=BarWidth, nodeColor=<rgba 1 0 0.2812 1>}) {constX=0.1}) {name=BarWidthSine, nodeColor=<rgba 0.9219 1 0 1>}
  NamedRerouteDeclaration(None=TextureObject() {samplerType=SAMPLERTYPE_Masks, texture=/Game/Global/Textures/Default/Noise/T_Generic_Perlin}) {name=NoiseRandomize, nodeColor=<rgba 1 0 0.0703 1>}
  NamedRerouteDeclaration(None=Constant() {r=0.0125}) {name=FakeFFTSpeedX, nodeColor=<rgba 1 0 0.4688 1>}
  NamedRerouteDeclaration(None=Constant() {r=3}) {name=FakeFFTPow, nodeColor=<rgba 0.9688 0 1 1>}
  n13 = NamedRerouteUsage()
  n14 = NamedRerouteUsage()
  n15 = NamedRerouteUsage()
  n16 = NamedRerouteUsage()
  n17 = Floor(None=Multiply(A=@n13, B=@n16))
  n18 = TextureSample(UVs=Multiply(A=Divide(A=@n17, B=@n16), B=Constant2Vector() {r=1}), Tex=NamedRerouteUsage()) {samplerType=SAMPLERTYPE_Masks, texture=&T_Generic_Perlin}
  n19 = Constant() {r=2}
  n20 = Add(
   A=NamedRerouteDeclaration(None=Constant() {r=0.05}) {name=FFTAdd, nodeColor=<rgba 0 1 0.6484 1>}
   B=Multiply(
    A=MaterialFunctionCall!(
     In TexCoords=@n18.R
     Fake FFT Noise Texture
     Fake FFT Noise Size
     Fake FFT Noise Speed X=@n14
     Fake FFT Noise Speed Y
     Fake FFT Noise Intensity Mult
     Fake FFT Noise Pow=@n15
    ) {materialFunction=/Game/Effects/Niagara/MasterMaterials/Functions/Mojo_MF/MF_FX_Fake_FFT_Noise}.Result Horizontal
    B=NamedRerouteDeclaration(None=@n19) {name=FFTIntensityMult, nodeColor=<rgba 0.4531 1 0 1>}
   ) {constB=2}
  )
  NamedRerouteDeclaration(None=Step(Y=ComponentMask(None=@n13) {g=true}, X=Saturate(None=@n20))) {name=FFTEQBarMask, nodeColor=<rgba 0.2422 1 0 1>}
  n21 = SphereMask(
   A=TextureCoordinate()
   B=Constant() {r=0.5}
   Radius=ScalarParameter() {defaultValue=0.5, parameterName=Radius}
   Hardness=ScalarParameter() {parameterName=Hardness}
  ) {hardnessPercent=0}
  n22 = DynamicParameter() {defaultValue=<rgba 3 1 0 1>, paramNames [Power, DepthFadeDistMult, AngleFade, CameraFade]}
  n23 = Clamp(None=@n22.Power) {clampMode=CMODE_ClampMin, maxDefault=100, minDefault=0.5}
  n24 = MaterialFunctionCall(Distance=ScalarParameter() {defaultValue=100, parameterName=NearCameraFadeDist}) {materialFunction=&Near_Camera_Fade}
  n25 = Multiply(A=@n22.DepthFadeDistMult, B=ScalarParameter() {defaultValue=100, parameterName=DepthFadeDist})
  n26 = Multiply(
   A=Multiply(
    A=DepthFade(Opacity=@n24.Result, FadeDistance=@n25)
    B=MaterialFunctionCall("FadeSensitivity (optional)"=@n22.AngleFade) {materialFunction=&FadeBasedOnViewAngle}.Result
   )
   B=MaterialFunctionCall(Distance=@n22.CameraFade) {materialFunction=&Near_Camera_Fade}.Result
  )
  n27 = Multiply(A=Multiply(A=Power(Base=@n21, Exp=@n23) {constExponent=3}, B=Multiply(A=@n1.A, B=@n26))) {desc=Optional boost}
  n28 = OneMinus(None=@n8)
  n29 = Saturate(None=Multiply(A=Power(Base=@n28) {constExponent=4}) {constB=4})
  n30 = NamedRerouteUsage()
  n31 = OneMinus(None=Constant() {r=0.001})
  n32 = ComponentMask(None=NamedRerouteUsage()) {g=true}
  n33 = NamedRerouteUsage()
  n34 = Constant2Vector() {r=1}
  n35 = MaterialFunctionCall(
   In TexCoords=TextureSample(UVs=Multiply(A=@n33, B=@n34), Tex=NamedRerouteUsage()) {samplerType=SAMPLERTYPE_Masks, texture=&T_Generic_Perlin}.R
   Fake FFT Noise Speed X=NamedRerouteUsage()
   Fake FFT Noise Pow=NamedRerouteUsage()
  ) {materialFunction=&MF_FX_Fake_FFT_Noise}
  n36 = NamedRerouteUsage()
  n37 = Saturate(None=Add(A=Multiply(A=@n35.Result Horizontal, B=@n36) {constB=2}, B=NamedRerouteUsage()))
  n38 = Multiply(A=NamedRerouteUsage(), B=SmoothStep(Min=@n32, Max=Add(A=@n32, B=@n31), Value=OneMinus(None=@n37)))
  n39 = Constant() {r=0.25}
  n40 = Constant() {r=1}
  n41 = Multiply(A=Multiply(
    A=Multiply(A=@n38, B=@n30)
    B=MaterialFunctionCall(Radius=@n39, Density=@n40) {materialFunction=&RadialGradientExponential}.RadialGradientExponential
   )) {constB=0.05}
  Saturate(None=Add(A=Multiply(A=@n27, B=@n29) {desc=bottom mask}, B=Saturate(None=@n41)))
 ]
}
