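# Niagara system for the Sparks Alchemy drum kit cosmetic, adding potion liquid, flames, smoke and magic glows to the kit's skeletal mesh.
# A pendulum constraint with inherited motion and beat-driven curl noise sloshes the liquid mesh, passing orientation, speed and turbulence to its material.
# Partitioned particles place flame ribbons, glows and smoke across eight sockets, and green and purple lit stars at two others.
# A front-end preview flag adjusts magic brightness.
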NiagaraSystem /SparksCosmetics/Instrument/DrumKit/Alchemy/Effects/NS_DrumKit_Alchemy
dependencies {
 usedDataInterfaces [
  &NiagaraDataInterfaceArrayColor
  &NiagaraDataInterfaceArrayFloat3
  &NiagaraDataInterfaceCurve
  &NiagaraDataInterfaceArrayFloat
  &NiagaraDataInterfaceArrayFloat2
  &NiagaraDataInterfaceArrayInt32
  &NiagaraDataInterfaceVector2DCurve
  &NiagaraDataInterfaceSpriteRendererInfo
  &NiagaraDataInterfaceSkeletalMesh
 ]
 usedDynamicInputs [
  &SelectVector2DFromArray
  &Vector2DFromCurve
  &RandomRangeVector2D
 ]
 usedModules [
  &self:SystemSpawnScript.SystemScriptSource.SystemScriptGraph.NiagaraNodeAssignment_0.SetVariables_#1
  &self:SystemSpawnScript.SystemScriptSource.SystemScriptGraph.NiagaraNodeAssignment_1.SetVariables_#2
  &self:SystemSpawnScript.SystemScriptSource.SystemScriptGraph.NiagaraNodeAssignment_2.SetVariables_#3
  &self:SystemSpawnScript.SystemScriptSource.SystemScriptGraph.NiagaraNodeAssignment_3.SetVariables_#4
  &SystemState
 ]
}
emitters {
 Alchemy_DrumKit_Potion_Liquid001 {
  emitterUpdateScript [
   &EmitterState(
    Inactive Response=::NewEnumerator1
    Life Cycle Mode=::NewEnumerator1
    Loop Behavior=::NewEnumerator1
    Loop Duration Mode=::NewEnumerator1
    UseLoopDelay=false
   )
   &SpawnBurst_Instantaneous(Use Spawn Probability=false, UseLoopCountLimit=false)
   /SetVariables(Emitter.Orientation:Quat4f=$Engine.Owner.Rotation) {isSetParameters=true}
   /SetVariables(Emitter.Position:Vector3f=/Niagara/DynamicInputs/Helpers/SimulationPosition()) {isSetParameters=true}
   /SetVariables(
    Emitter.RotationVelocity:Vector3f=&Divide_Vector(
     A:Vector3f=&Subtract_Vector(
      A=&RotateVectorByQuaternion(Vector To Rotate=<0 0 1>, Quaternion=$Emitter.Orientation)
      B=&RotateVectorByQuaternion(Vector To Rotate=<0 0 1>, Quaternion=$Emitter.Previous.Orientation)
     )
     B:Vector3f=&VectorFromFloat(Value=$Engine.DeltaTime)
    )
    Emitter.Velocity:Vector3f=&Divide_Vector(
     A:Vector3f=&Subtract_Vector(A=$Emitter.Position, B=$Emitter.Previous.Position)
     B:Vector3f=&VectorFromFloat(Value=$Engine.DeltaTime)
    )
    Emitter.Viscosity=0.5
   ) {isSetParameters=true}
   /SetVariables(Emitter.InheritSourceMovement:NiagaraFloat=&Lerp_Float(A=0, B=0.9, Alpha=$Emitter.Viscosity)) {isSetParameters=true}
  ]
  particleSpawnScript [
   &InitializeParticle(
    Color Mode=::NewEnumerator0
    Mesh Scale=$System.LiquidScale
    Mesh Scale Mode=::NewEnumerator4
    Position Offset=&Multiply_VectorByFloat(Vector=<0 0 1>, Float=16)
    UsePositionOffset=false
    Write Material Random=true
   )
   &SkeletalMeshLocation(
    Bone / Socket Index=11
    Bone / Socket Sampling Mode=::NewEnumerator5
    MaskSpawnGroup=false
    Mesh Sampling Type=::NewEnumerator6
    Orientation Sampling=::NewEnumerator1
    Position Sampling=::NewEnumerator2
    Skeletal Mesh=$System.Skeletal Mesh
   )
  ]
  particleUpdateScript [
   &ParticleState(Kill Particles When Lifetime Has Elapsed=false, Let Infinitely Lived Particles Die When Emitter Deactivates=true)
   &ScaleMeshSize(Scale Factor=$Engine.Owner.Scale)
   /SetVariables(Particles.savedPosition:NiagaraPosition=$Particles.Position) {isSetParameters=true}
   &SkeletalMeshLocation(
    Bone / Socket Index=11
    Bone / Socket Sampling Mode=::NewEnumerator5
    MaskSpawnGroup=false
    Mesh Sampling Type=::NewEnumerator6
    Orientation Sampling=::NewEnumerator1
    Skeletal Mesh=$System.Skeletal Mesh
   ) {module=SkeletalMeshLocation001}
   /SetVariables(Particles.SampledScale:Vector3f=$Particles.SkeletalMeshLocation001.SampledScale) {isSetParameters=true}
   /SetVariables(Particles.tempPos:NiagaraPosition=&SubtractVectorFromPosition(Position=$Particles.Position, Vector=Linked() {linkedVariable {name=Engine.Emitter.SimulationPosition, type=NiagaraPosition}})) {isSetParameters=true}
   &ScaleMeshSize(Scale Factor=$Particles.SampledScale) {module=ScaleMeshSize001}
   /SetVariables(Particles.Position=$Particles.savedPosition) {isSetParameters=true}
   /SetVariables(Particles.Position_Origin:NiagaraPosition=$Engine.Emitter.SimulationPosition) {enabled=false, isSetParameters=true}
   /SetVariables(Particles.Position_Origin:NiagaraPosition=&AddVectorToPosition(Position=$Engine.Emitter.SimulationPosition, Vector=Linked() {linkedVariable {name=Particles.tempPos, type=NiagaraPosition}})) {isSetParameters=true}
   &PendulumSetup(
    Module.Enable Spring Driver=true
    Pendulum Length:NiagaraFloat=&Multiply_Float(
     A=&Lerp_Float(A=10, B=10, Alpha=&MakeCustomFloatFromBool(Bool=$User.bIsFrontEndPreview, True Float=1, False Float=0))
     B=&MakeFloatFromVector(Vector=$System.LiquidScale, Channel=::NewEnumerator2)
    )
    Pendulum Pivot:NiagaraPosition=$Particles.Position_Origin
    Pendulum Rest Axis:Vector3f=&NormalizeVector(
     Vector To Normalize=&Add_Vector(
      A=<0 0 -1>
      B=&Lerp_Vector(
       A=<0 0 0>
       B=&TransformVector(Vector=<0 0 -1>, Destination Space=::Simulation)
       Alpha=$System.Timeline.Float
      )
     )
     Safe Normalize=true
     Safe Fallback Threshold=0
     Safe Fallback Vector=<0 0 0.001>
    )
    "Spring Driven Constraint (Tightness)":NiagaraFloat=&MakeCustomFloatFromBool(Bool=$User.IsFilling, True Float=1, False Float=0)
   )
   &GravityForce(Gravity=&Lerp_Vector(A=<0 0 -980>, B=<0 0 -19.6>, Alpha=$Emitter.Viscosity))
   &Drag(Drag=&Lerp_Float(A=0, B=5, Alpha=$Emitter.Viscosity), Use Linear Drag=true, Use Rotational Drag=false)
   &InheritSourceMovement(
    Applied Position Scale=&VectorFromFloat(Value=$Emitter.InheritSourceMovement)
    ApplyForce=false
    ApplyPosition=true
    ApplyPositionFalloff=false
    Source Velocity:Vector3f=$Emitter.Velocity
   )
   &CurlNoiseForce(
    Noise Strength=&Multiply_Float(
     A=500
     B=&Lerp_Float(
      A=0
      B=1
      Alpha=&FloatFromCurve(
       FloatCurve=DataInterface() {propertyValues {Curve {keys [
           {arriveTangent=9.0725, interpMode=RCIM_Cubic, time=0, value=0}
           <at 0.1 1 RCIM_Cubic>
           {arriveTangent=-0.0088, interpMode=RCIM_Cubic, leaveTangent=-0.0088, tangentMode=RCTM_User, time=1, value=0}
          ]}}}
       CurveIndex=$NPC.NPC_Sparks_Audio_Parameters001.BeatLifetime
      )
     )
    )
    Pan Noise=true
    Pan Noise Field=<0 0 100>
   )
   &CurlNoiseForce(Noise Strength=65, Pan Noise=true, Pan Noise Field=<0 0 2>) {module=CurlNoiseForce001}
   &SolveForcesAndVelocity(Clamp Velocity=false, Limit Acceleration=false, Speed Limit=200)
   &PendulumConstraint()
   &CalculateAccurateVelocity(Previous Position:NiagaraPosition=&Add_Vector(
     A=Linked() {linkedVariable {name=Particles.Previous.Position, type=NiagaraPosition}}
     B=&Multiply_VectorByFloat(Vector=$Emitter.Velocity, Float=&Multiply_Float(A=$Emitter.InheritSourceMovement, B=$Engine.DeltaTime))
    ))
   /Niagara/Modules/Update/Orientation/V2/OrientMeshToVector(
    Look At Direction:Vector3f=&NormalizeVector(
     Vector To Normalize=&Subtract_Vector(
      A=Linked() {linkedVariable {name=Particles.Position_Origin, type=NiagaraPosition}}
      B=Linked() {linkedVariable {name=Particles.Position, type=NiagaraPosition}}
     )
     Safe Normalize=true
     Safe Fallback Threshold=0
     Safe Fallback Vector=<0 0 0.001>
    )
    Override Quat=false
    Reference Axis Vector:Vector3f=<0 0 1>
    UseOptionalSideAxis=false
    UseOptionalUpAxis=false
   )
   &InterpolateOverTime(
    Rate Of Change=2
    Target Float Value:NiagaraFloat=&Max_Float(
     A:NiagaraFloat=&RemapRange(
      Input Value To Remap=&DotProduct(
       Vector A=<0 0 -1>
       Vector B=&ReturnMeshOrientationAxis(Mesh Orientation Axis=::ZAxis, Mesh Orientation Quat=$Particles.MeshOrientation)
      )
      Input Min=-1
      Input Max=0
      Output Min=0
      Output Max=1
     )
     B:NiagaraFloat=&RemapRange(
      Input Value To Remap=&Divide_Float(
       A=&Subtract_Float(A=$Particles.DistanceTraveled, B=$Particles.Previous.DistanceTraveled)
       B=&MakeFloatFromVector(Vector=$System.LiquidScale, Channel=::NewEnumerator2)
      )
      Input Min=2
      Input Max=4
      Output Min=0
      Output Max=1
     )
    )
   )
   /SetVariables(Particles.LiquidSpeed:NiagaraFloat=&ClampFloat(Float=$Particles.InterpolateOverTime.InterpolateOverTime_Float.Moving Average)) {isSetParameters=true}
   &Timeline(Play=&SetBoolByFloatComparison(A=$Particles.LiquidSpeed, B=0.333), Play Rate=3, Timeline End=10)
   /SetVariables(Particles.Turbulence:NiagaraFloat=$Particles.Timeline.NormalizedTime) {isSetParameters=true}
   /Niagara/Modules/Update/Utility/IncrementOverTime(Rate of Change:NiagaraFloat=&Lerp_Float(A=1, B=1.25, Alpha=$Particles.LiquidSpeed))
   /SetVariables(
    Particles.DynamicMaterialParameter:Vector4f=&MakeFloat4FromVectorAndFloat(
     "Vector (RGB)"=&ReturnMeshOrientationAxis(Mesh Orientation Axis=::ZAxis, Mesh Orientation Quat=$Particles.MeshOrientation)
     "Float (Alpha)"=&Lerp_Float(A=0, B=0.425, Alpha=$System.Timeline.NormalizedTime)
    )
    Particles.DynamicMaterialParameter1:Vector4f=&MakeFloat4FromVectorAndFloat(
     "Vector (RGB)"=&ReturnMeshOrientationAxis(Mesh Orientation Axis=::XAxis, Mesh Orientation Quat=$Particles.MeshOrientation)
     "Float (Alpha)"=$Particles.IncrementOverTime.Counter
    )
    Particles.DynamicMaterialParameter2:Vector4f=&MakeVector4(X=$Particles.LiquidSpeed, Y=$Particles.Turbulence, Z=0, W=0)
   ) {isSetParameters=true}
   &CameraOffset()
   &SkeletalMeshLocation(
    Bone / Socket Index=13
    MaskSpawnGroup=false
    Mesh Sampling Type=::NewEnumerator5
    Orientation Sampling=::NewEnumerator1
    Position Sampling=::NewEnumerator2
    Sampled Position Offset Coordinate Space:ENiagaraSkelMeshTransforms=::NewEnumerator1
    Skeletal Mesh=$System.Skeletal Mesh
    Socket Sampling Mode=::NewEnumerator5
   ) {module=SkeletalMeshLocation002}
   /SetVariables(Particles.Position_Origin:NiagaraPosition=&AddVectorToPosition(
     Position=$Particles.SkeletalMeshLocation002.SampledPosition
     Vector=&Multiply_VectorByFloat(Vector=Linked() {linkedVariable {name=Engine.Emitter.SimulationPosition, type=NiagaraPosition}}, Float=0)
    )) {enabled=false, isSetParameters=true}
   /SetVariables(Particles.Position_Origin:NiagaraPosition=$Particles.SkeletalMeshLocation002.SampledPosition) {enabled=false, isSetParameters=true}
   /SetVariables(Particles.LocalOrientation:Quat4f=$Particles.SkeletalMeshLocation002.SampledRotation) {isSetParameters=true}
  ]
  rendererClasses [&NiagaraMeshRendererProperties]
  renderers [
   {rendererClass=&NiagaraMeshRendererProperties, rendererIndex=0}
   {rendererClass=&NiagaraMeshRendererProperties, rendererIndex=1}
  ]
 }
 FlameGlow {
  emitterUpdateScript [
   &EmitterState(Life Cycle Mode=::NewEnumerator1, Loop Behavior=::NewEnumerator1, Loop Delay=0.2, Loop Duration Mode=::NewEnumerator1, UseLoopDelay=true)
   &SpawnBurst_Instantaneous(Spawn Count=8, Use Spawn Probability=false, UseLoopCountLimit=false)
  ]
  particleSpawnScript [
   &PartitionParticles(Number Of Partitions=8)
   &InitializeParticle(
    Color=<rgba 1 0.7071 0.3384 0.2>
    Lifetime=2
    Ribbon Width=50
    Ribbon Width Mode=::NewEnumerator1
    Sprite Size=&Multiply_Vector2D(
     A=<10 15>
     B=&Vector2DFromFloat(Value=&SelectFloatFromArray(
       Direct Array Index=$Particles.PartitionParticles.Partition
       Float Selection Array=DataInterface() {propertyValues {FloatData [1,1,1,1,1,1,1,1]}}
      ))
    )
    Sprite Size Mode=::NewEnumerator4
    UsePositionOffset=false
    Write Material Random=true
   )
  ]
  particleUpdateScript [
   &ParticleState(Kill Particles When Lifetime Has Elapsed=false, Let Infinitely Lived Particles Die When Emitter Deactivates=true)
   &SkeletalMeshLocation(
    Bone / Socket Index=$Particles.PartitionParticles.Partition
    MaskSpawnGroup=false
    Mesh Sampling Type=::NewEnumerator5
    Sampled Position Offset=&SelectVectorFromArray(
     Direct Array Index=$Particles.PartitionParticles.Partition
     Vector Selection Array=DataInterface() {propertyValues {FloatData [<0 0 1>, <0 0 0>, <-13.5 -10.5 10.5>, <0 0 1>, <0 0 1>, <0 0 1>, <0 0 1>, <0 0 1>]}}
    )
    Sampled Position Offset Coordinate Space:ENiagaraSkelMeshTransforms=::NewEnumerator2
    Skeletal Mesh=$System.Skeletal Mesh
    Socket Sampling Mode=::NewEnumerator5
   ) {module=SkeletalMeshLocation001}
   &Drag(
    Drag=&Multiply_Float(
     A=10
     B=&FloatFromCurve(FloatCurve=DataInterface() {propertyValues {Curve {keys [{tangentMode=RCTM_User, time=0, value=1}, {tangentMode=RCTM_User, time=0.7, value=1}]}}})
    )
    Use Linear Drag=true
    Use Rotational Drag=true
   )
   &DynamicMaterialParameters(
    Index 0 Param 1=5
    Module.Index0Param0Write=true
    Module.Index0Param1Write=true
    Module.Index0Param2Write=true
    Module.Index0Param3Write=true
    Module.Index1Param0Write=true
    Module.Index1Param1Write=true
    Module.Index1Param2Write=true
    Module.Index1Param3Write=true
    Module.Index2Param0Write=true
    Module.Index2Param1Write=true
    Module.Index2Param2Write=true
    Module.Index2Param3Write=true
    Write Parameter Index 1=true
    Write Parameter Index 2=true
   )
   &SolveForcesAndVelocity(Clamp Velocity=false, Limit Acceleration=false)
   &FN_Light_Attributes(
    Color=$Particles.Color
    Create Color Attribute=true
    Create Exponent Attribute=true
    Create Radius Attribute=true
    Enable Particle Light Enable=true
    Exponent=50
    Light Enabled=true
    Radius=&Multiply_Float(A=10, B=&MakeFloatFromVector(Vector=$Engine.Owner.Scale, Channel=::NewEnumerator0))
    Scale Color=&Multiply_Float(A=10, B=&MakeFloatFromVector(Vector=$Engine.Owner.Scale, Channel=::NewEnumerator0))
   )
   &CameraOffset(Camera Offset Amount=1)
   &ScaleSpriteSize(
    Scale Sprite Size Mode=::NewEnumerator0
    Uniform Scale Factor=&MakeFloatFromVector(Vector=$Engine.Owner.Scale, Channel=::NewEnumerator0)
   )
  ]
  rendererClasses [&NiagaraSpriteRendererProperties, &NiagaraLightRendererProperties]
  renderers [
   {rendererClass=&NiagaraSpriteRendererProperties}
   {rendererClass=&NiagaraLightRendererProperties, rendererIndex=1}
  ]
 }
 FlameRibbon {
  emitterUpdateScript [
   &EmitterState(Life Cycle Mode=::NewEnumerator1, Loop Behavior=::NewEnumerator1, Loop Delay=0.2, Loop Duration Mode=::NewEnumerator1, UseLoopDelay=true)
   &SpawnRate(
    SpawnRate=&Multiply_Float(A=100, B=&MakeFloatFromVector(Vector=$Engine.Owner.Scale, Channel=::NewEnumerator0))
    Use Spawn Probability=false
   )
  ]
  particleSpawnScript [
   &PartitionParticles(Number Of Partitions=8)
   &InitializeParticle(
    Color=&SelectLinearColorFromArray(
     Array Sampling Mode=::NewEnumerator1
     Direct Array Index=$Particles.PartitionParticles.Partition
     Color Selection Array=DataInterface() {propertyValues {ColorData [<rgba 1 1 1 1>, <rgba 1 1 1 1>, <rgba 1 1 1 1>, <rgba 1 1 1 1>, <rgba 1 1 1 1>, <rgba 1 1 1 1>, <rgba 1 1 1 1>, <rgba 1 1 1 1>]}}
     Random Seed=?
    )
    Ribbon Width=50
    Ribbon Width Mode=::NewEnumerator1
    Sprite Size Mode=::NewEnumerator3
    Uniform Sprite Size=1
    UsePositionOffset=false
    Write Material Random=true
   )
  ]
  particleUpdateScript [
   &ParticleState()
   &SkeletalMeshLocation(
    Bone / Socket Index=$Particles.PartitionParticles.Partition
    MaskSpawnGroup=false
    Mesh Sampling Type=::NewEnumerator5
    Orientation Sampling=::NewEnumerator1
    Position Sampling=::NewEnumerator1
    Sampled Position Offset=&SelectVectorFromArray(
     Direct Array Index=$Particles.PartitionParticles.Partition
     Vector Selection Array=DataInterface() {propertyValues {FloatData [<0 0 -2>, <0 0 -3>, <-13.5 -10.5 7.5>, <0 0 -2>, <0 0 -2>, <0 0 -2>, <0 0 -2>, <0 0 -2>]}}
    )
    Sampled Position Offset Coordinate Space:ENiagaraSkelMeshTransforms=::NewEnumerator2
    Skeletal Mesh=$System.Skeletal Mesh
    Socket Sampling Mode=::NewEnumerator5
   ) {module=SkeletalMeshLocation001}
   &GravityForce(Gravity=<0 0 15>)
   /SetVariables(Particles.RibbonWidth=&FloatFromCurve(
     FloatCurve=DataInterface() {propertyValues {Curve {keys [
         {arriveTangent=5.7315, interpMode=RCIM_Cubic, leaveTangent=5.7315, tangentMode=RCTM_User, time=0, value=0}
         {arriveTangent=-2.9308, interpMode=RCIM_Cubic, leaveTangent=-2.9308, tangentMode=RCTM_User, time=0.35, value=0.9847}
         {arriveTangent=-0.8966, interpMode=RCIM_Cubic, leaveTangent=-0.8966, tangentMode=RCTM_User, time=1, value=0}
        ]}}}
     Scale Curve=&Multiply_Float(A=4.5, B=&MakeFloatFromVector(Vector=$Engine.Owner.Scale, Channel=::NewEnumerator0))
    )) {isSetParameters=true}
   &DynamicMaterialParameters(
    Module.Index0Param0Write=true
    Module.Index0Param1Write=true
    Module.Index0Param2Write=true
    Module.Index0Param3Write=true
    Module.Index1Param0Write=true
    Module.Index1Param1Write=true
    Module.Index1Param2Write=true
    Module.Index1Param3Write=true
    Module.Index2Param0Write=true
    Module.Index2Param1Write=true
    Module.Index2Param2Write=true
    Module.Index2Param3Write=true
    Write Parameter Index 1=true
    Write Parameter Index 2=true
   )
   &SolveForcesAndVelocity(Clamp Velocity=false, Limit Acceleration=false)
   &CameraOffset(Camera Offset Amount=1)
  ]
  rendererClasses [
   &NiagaraRibbonRendererProperties
   &NiagaraSpriteRendererProperties
   &NiagaraMeshRendererProperties
  ]
  renderers [
   {rendererClass=&NiagaraRibbonRendererProperties}
   {rendererClass=&NiagaraSpriteRendererProperties, rendererIndex=1}
   {rendererClass=&NiagaraMeshRendererProperties, rendererIndex=2}
  ]
 }
 MagicGlowAndBigLight {
  emitterUpdateScript [
   &EmitterState(Life Cycle Mode=::NewEnumerator1, Loop Behavior=::NewEnumerator1, Loop Duration Mode=::NewEnumerator1, UseLoopDelay=false)
   &SpawnBurst_Instantaneous(Spawn Count=2, Use Spawn Probability=false, UseLoopCountLimit=false)
  ]
  particleSpawnScript [
   &PartitionParticles()
   &InitializeParticle(
    Color=&SelectLinearColorFromArray(
     Array Sampling Mode=::NewEnumerator1
     Direct Array Index=$Particles.PartitionParticles.Partition
     Color Selection Array=DataInterface() {propertyValues {ColorData [<rgba 0.8431 5 1.3887 0.9>, <rgba 3.4249 0.3506 5 0.9>]}}
     Random Seed=?
    )
    Lifetime=&RandomRangeFloat(Minimum=1, Maximum=2)
    Position Mode=::NewEnumerator0
    Sprite Size Mode=::NewEnumerator3
    Uniform Sprite Size=150
    Write Material Random=true
   ) {module=InitializeParticle001}
   &SkeletalMeshLocation(
    Bone / Socket Index=&SelectIntFromArray(Direct Array Index=$Particles.PartitionParticles.Partition, Int Selection Array=DataInterface() {propertyValues {IntData [8,9]}})
    Bone / Socket Sampling Mode=::NewEnumerator5
    MaskSpawnGroup=false
    Mesh Sampling Type=::NewEnumerator6
    Sampled Position Offset=&SelectVectorFromArray(
     Direct Array Index=$Particles.PartitionParticles.Partition
     Vector Selection Array=DataInterface() {propertyValues {FloatData [<2 0 0>, <-2.56 0 0>]}}
    )
    Sampled Position Offset Coordinate Space:ENiagaraSkelMeshTransforms=::NewEnumerator1
    Skeletal Mesh=$System.Skeletal Mesh
   ) {module=SkeletalMeshLocation003}
  ]
  particleUpdateScript [
   &ParticleState(
    Kill Particles When Lifetime Has Elapsed=false
    Let Infinitely Lived Particles Die When Emitter Deactivates=true
    Loop Particles Lifetime=true
   )
   &SkeletalMeshLocation(
    Bone / Socket Index=&SelectIntFromArray(Direct Array Index=$Particles.PartitionParticles.Partition, Int Selection Array=DataInterface() {propertyValues {IntData [8,9]}})
    Bone / Socket Sampling Mode=::NewEnumerator5
    MaskSpawnGroup=false
    Mesh Sampling Type=::NewEnumerator6
    Sampled Position Offset=&SelectVectorFromArray(
     Direct Array Index=$Particles.PartitionParticles.Partition
     Vector Selection Array=DataInterface() {propertyValues {FloatData [<2 0 0>, <-2.56 0 0>]}}
    )
    Sampled Position Offset Coordinate Space:ENiagaraSkelMeshTransforms=::NewEnumerator1
    Skeletal Mesh=$System.Skeletal Mesh
   ) {module=SkeletalMeshLocation002}
   &DynamicMaterialParameters(
    Index 0 Param 1=4
    Index 0 Param 2=100
    Index 0 Param 3=4
    Index 0 Param 4=3
    Module.Index0Param0Write=true
    Module.Index0Param1Write=true
    Module.Index0Param2Write=true
    Module.Index0Param3Write=true
   )
   /SetVariables(
    Particles.GlowColor=&MultiplyLinearColor(Linear Color=$Particles.Color, Scale Factor=<0.2 0.2 0.2 1>)
    Particles.GlowSize=&Multiply_Vector2D(
     A=$Particles.SpriteSize
     B=&Vector2DFromFloat(Value=&MakeFloatFromVector(Vector=$Engine.Owner.Scale, Channel=::NewEnumerator0))
    )
   ) {isSetParameters=true, module=SetVariables_#5}
   &FN_Light_Attributes(
    Color=$Particles.Color
    Create Color Attribute=true
    Create Exponent Attribute=true
    Create Radius Attribute=true
    Enable Particle Light Enable=true
    Light Enabled=true
    Radius=100
    Scale Color=&Multiply_Float(A=50, B=&MakeFloatFromVector(Vector=$Engine.Owner.Scale, Channel=::NewEnumerator0))
   )
   &ScaleColor(
    Scale Alpha=&MakeFloatFromVector(
     Vector=&VectorFromFloat(Value=&Lerp_Float(
       A=1.25
       B=1
       Alpha=&MakeCustomFloatFromBool(Bool=$User.bIsFrontEndPreview, True Float=1, False Float=0)
      ))
     Channel=::NewEnumerator0
    )
    Scale RGB=&VectorFromFloat(Value=&Lerp_Float(
      A=0.8
      B=0.75
      Alpha=&MakeCustomFloatFromBool(Bool=$User.bIsFrontEndPreview, True Float=1, False Float=0)
     ))
    ScaleA=true
    ScaleRGB=true
   ) {module=ScaleColor001}
   &CameraOffset(Camera Offset Amount=-1)
  ]
  rendererClasses [&NiagaraSpriteRendererProperties, &NiagaraLightRendererProperties]
  renderers [
   {rendererClass=&NiagaraSpriteRendererProperties}
   {rendererClass=&NiagaraLightRendererProperties, rendererIndex=1}
  ]
 }
 MagicStars {
  emitterUpdateScript [
   &EmitterState()
   &SpawnRate(
    SpawnRate=&FloatFromCurve(
     FloatCurve=DataInterface() {propertyValues {Curve {keys [
         {arriveTangent=9.0725, interpMode=RCIM_Cubic, time=0, value=0}
         {arriveTangent=-0.1629, interpMode=RCIM_Cubic, leaveTangent=-0.1629, tangentMode=RCTM_User, time=0.1408, value=1}
         <at 0.4 0.35 RCIM_Cubic>
         {arriveTangent=-0.5833, interpMode=RCIM_Cubic, leaveTangent=-0.5833, time=0.7, value=1}
         {interpMode=RCIM_Cubic, leaveTangent=-0.0405, time=1, value=0}
        ]}}}
     CurveIndex=$Emitter.NormalizedLoopAge
     Scale Curve=15
    )
    Use Spawn Probability=true
   )
  ]
  particleSpawnScript [
   &PartitionParticles()
   &InitializeParticle(
    Color=&SelectLinearColorFromArray(
     Array Sampling Mode=::NewEnumerator1
     Direct Array Index=$Particles.PartitionParticles.Partition
     Color Selection Array=DataInterface() {propertyValues {ColorData [<rgba 0.8431 5 1.3887 0.9>, <rgba 3.4249 0.3506 5 0.9>]}}
     Random Seed=?
    )
    Lifetime=&RandomRangeFloat(Minimum=1, Maximum=2)
    Position Mode=::NewEnumerator0
    Sprite Rotation Mode=::NewEnumerator2
    Sprite Size Mode=::NewEnumerator3
    Uniform Sprite Size=&RandomRangeFloat(Minimum=1.5, Maximum=4)
    Write Material Random=true
   ) {module=InitializeParticle001}
   &ShapeLocation(
    Apply To Particle Position=::NewEnumerator1
    MaskSpawnGroup=false
    Non Uniform Scale=<0 1 1>
    Override Seed=false
    Shape Origin=&VectorFromFloat(Value=0)
    Sphere Radius=8
    UseOwnerScale=true
   ) {module=ShapeLocation001}
   &AddVelocity(Velocity=&RandomRangeVector(Minimum=<2 0 5>, Maximum=<-2 1 25>))
  ]
  particleUpdateScript [
   &ParticleState()
   &SkeletalMeshLocation(
    Bone / Socket Index=&SelectIntFromArray(Direct Array Index=$Particles.PartitionParticles.Partition, Int Selection Array=DataInterface() {propertyValues {IntData [8,9]}})
    Bone / Socket Sampling Mode=::NewEnumerator5
    MaskSpawnGroup=false
    Mesh Sampling Type=::NewEnumerator6
    Position Sampling=::NewEnumerator1
    Sampled Position Offset=$Particles.Initial.ShapeLocation001.ShapePosition
    Skeletal Mesh=$System.Skeletal Mesh
   )
   &SpriteRotationRate(Rotation Rate=&RandomRangeFloat(Minimum=-90, Maximum=90))
   &ScaleColor(
    Scale Alpha=&FloatFromCurve(
     FloatCurve=DataInterface() {propertyValues {Curve {keys [
         {arriveTangent=9.0725, interpMode=RCIM_Cubic, leaveTangent=9.0725, tangentMode=RCTM_User, time=0, value=0}
         {arriveTangent=0.1899, interpMode=RCIM_Cubic, leaveTangent=0.1899, tangentMode=RCTM_User, time=0.2, value=0.9847}
         {arriveTangent=-0.0775, interpMode=RCIM_Cubic, leaveTangent=-0.0775, tangentMode=RCTM_User, time=0.3, value=1}
         {arriveTangent=-0.0405, interpMode=RCIM_Cubic, leaveTangent=-0.0405, tangentMode=RCTM_User, time=1, value=0}
        ]}}}
     Scale Curve=0.5
    )
    ScaleA=true
    ScaleRGB=true
   )
   &ScaleSpriteSize(Scale Factor=&Vector2DFromFloat(Value=&FloatFromCurve(FloatCurve=DataInterface() {propertyValues {Curve {keys [
          {arriveTangent=3.9776, interpMode=RCIM_Cubic, leaveTangent=3.9776, tangentMode=RCTM_User, time=0, value=0}
          {arriveTangent=0.0107, interpMode=RCIM_Cubic, leaveTangent=0.0107, tangentMode=RCTM_User, time=0.4, value=0.9847}
          {arriveTangent=-2.2605, interpMode=RCIM_Cubic, leaveTangent=-2.2605, tangentMode=RCTM_User, time=1, value=0.2}
         ]}}})))
   &DynamicMaterialParameters(
    Index 0 Param 3=&FloatFromCurve(FloatCurve=DataInterface() {propertyValues {Curve {keys [{arriveTangent=-53.9199, time=0, value=0}, {leaveTangent=0.2723, time=0.25, value=0}]}}})
    Index 0 Param 4=&FloatFromCurve(FloatCurve=DataInterface() {propertyValues {Curve {keys [<at 0 1>,<at 1 0>]}}})
    Index 1 Param 1=&FloatFromCurve(FloatCurve=DataInterface() {propertyValues {Curve {keys [<at 0 0>, {arriveTangent=4.2257, interpMode=RCIM_Cubic, leaveTangent=4.2257, tangentMode=RCTM_User, time=0.25, value=0}]}}})
    Index 1 Param 2=&FloatFromCurve(FloatCurve=DataInterface() {propertyValues {Curve {keys [<at 0 0>]}}})
    Index 1 Param 3=&FloatFromCurve(FloatCurve=DataInterface() {propertyValues {Curve {keys [{interpMode=RCIM_Cubic, leaveTangent=6.3917, tangentMode=RCTM_Break, time=0, value=0}]}}})
    Index 1 Param 4=&FloatFromCurve(FloatCurve=DataInterface() {propertyValues {Curve {keys [{arriveTangent=-0.136, leaveTangent=-0.136, time=0, value=1}]}}})
    Index 2 Param 1=&FloatFromCurve(FloatCurve=DataInterface() {propertyValues {Curve {keys [
         {arriveTangent=2.646, interpMode=RCIM_Cubic, leaveTangent=2.646, tangentMode=RCTM_User, time=0, value=0}
         {leaveTangent=-0.0719, time=1, value=1}
        ]}}})
    Index 2 Param 2=&FloatFromCurve(
     FloatCurve=DataInterface() {propertyValues {Curve {keys [
         {arriveTangent=1.9056, time=0, value=0.2}
         {arriveTangent=0.0259, interpMode=RCIM_Cubic, leaveTangent=0.0259, tangentMode=RCTM_User, time=1, value=1.4}
        ]}}}
     Scale Curve=&RandomRangeFloat(Minimum=0.5, Maximum=1)
    )
    Index 2 Param 3=&FloatFromCurve(FloatCurve=DataInterface() {propertyValues {Curve {keys [
         {interpMode=RCIM_Cubic, leaveTangent=2.3064, tangentMode=RCTM_Break, time=0, value=1}
         {interpMode=RCIM_Cubic, leaveTangent=11.1233, tangentMode=RCTM_Break, time=1, value=2}
        ]}}})
    Index 2 Param 4=&FloatFromCurve(
     FloatCurve=DataInterface() {propertyValues {Curve {keys [
         {arriveTangent=-0.6867, interpMode=RCIM_Cubic, leaveTangent=-0.6867, tangentMode=RCTM_User, time=0, value=1}
         {interpMode=RCIM_Cubic, leaveTangent=-1.2956, tangentMode=RCTM_Break, time=1, value=0.75}
        ]}}}
     Scale Curve=0.1
    )
    Module.Param0WriteEnabled=true
    Module.Param0WriteEnabled1=false
    Module.Param0WriteEnabled2=true
    Module.Param1WriteEnabled=false
    Module.Param1WriteEnabled1=true
    Module.Param1WriteEnabled2=true
    Module.Param2WriteEnabled=true
    Module.Param2WriteEnabled1=true
    Module.Param2WriteEnabled2=true
    Module.Param3WriteEnabled=true
    Module.Param3WriteEnabled1=true
    Module.Param3WriteEnabled2=true
    Write Parameter Index 1=true
    Write Parameter Index 2=true
   )
   &CurlNoiseForce(Noise Strength=3, Pan Noise=false)
   &CurlNoiseForce(
    Noise Strength=&FloatFromCurve(
     FloatCurve=DataInterface() {propertyValues {Curve {keys [
         {interpMode=RCIM_Cubic, tangentMode=RCTM_User, time=0, value=0}
         {arriveTangent=0.1375, interpMode=RCIM_Cubic, leaveTangent=0.1375, tangentMode=RCTM_User, time=1, value=1}
        ]}}}
     Scale Curve=100
    )
    Pan Noise=true
    Pan Noise Field=<0 0 1>
   ) {module=CurlNoiseForce001}
   &SolveForcesAndVelocity(Clamp Velocity=false, Limit Acceleration=false)
   /SetVariables(
    Particles.GlowColor=&MultiplyLinearColor(Linear Color=$Particles.Color, Scale Factor=<0.1 0.1 0.1 0.15>)
    Particles.GlowSize=&Multiply_Vector2D(
     A=$Particles.SpriteSize
     B=&Vector2DFromFloat(Value=&Multiply_Float(A=1.35, B=&MakeFloatFromVector(Vector=$Engine.Owner.Scale, Channel=::NewEnumerator0)))
    )
   ) {isSetParameters=true, module=SetVariables_#5}
   &FN_Light_Attributes(
    Color=$Particles.Color
    Create Color Attribute=true
    Create Exponent Attribute=true
    Create Radius Attribute=true
    Enable Particle Light Enable=true
    Light Enabled=true
    Radius=&Multiply_Float(A=10, B=&MakeFloatFromVector(Vector=$Engine.Owner.Scale, Channel=::NewEnumerator1))
    Scale Color=&Multiply_Float(A=5, B=&MakeFloatFromVector(Vector=$Engine.Owner.Scale, Channel=::NewEnumerator1))
   )
   &ScaleColor(
    Scale Alpha=&MakeFloatFromVector(
     Vector=&VectorFromFloat(Value=&Lerp_Float(A=2, B=1, Alpha=&MakeCustomFloatFromBool(Bool=$User.bIsFrontEndPreview, True Float=1, False Float=0)))
     Channel=::NewEnumerator0
    )
    Scale RGB=&VectorFromFloat(Value=&Lerp_Float(A=10, B=3, Alpha=&MakeCustomFloatFromBool(Bool=$User.bIsFrontEndPreview, True Float=1, False Float=0)))
    ScaleA=true
    ScaleRGB=true
   ) {module=ScaleColor001}
   &ScaleSpriteSize(
    Scale Sprite Size Mode=::NewEnumerator0
    Uniform Scale Factor=&MakeFloatFromVector(Vector=$Engine.Owner.Scale, Channel=::NewEnumerator0)
   ) {module=ScaleSpriteSize001}
  ]
  rendererClasses [&NiagaraSpriteRendererProperties, &NiagaraLightRendererProperties]
  renderers [
   {rendererClass=&NiagaraSpriteRendererProperties}
   {rendererClass=&NiagaraSpriteRendererProperties, rendererIndex=1}
   {rendererClass=&NiagaraLightRendererProperties, rendererIndex=2}
  ]
 }
 Smoke_Detail {
  emitterUpdateScript [
   &EmitterState(Inactive Response=::NewEnumerator1, Life Cycle Mode=::NewEnumerator1, Loop Delay=0.4, Loop Duration=1, UseLoopDelay=false)
   &SpawnBurst_Instantaneous(Spawn Count=7, Use Spawn Probability=false, UseLoopCountLimit=false)
   &SpawnRate(SpawnRate=20, Use Spawn Probability=false) {module=SpawnRate001}
  ]
  particleSpawnScript [
   &PartitionParticles(Number Of Partitions=8)
   &InitializeParticle(
    Color=<rgba 0.2353 0.2353 0.2353 0.75>
    Lifetime Max=1.25
    Lifetime Min=0.5
    Lifetime Mode=::NewEnumerator1
    Sprite Rotation Mode=::NewEnumerator2
    Sprite Size Max=<0.7 0.7>
    Sprite Size Min=<0.5 0.5>
    Sprite Size Mode=::NewEnumerator2
    UsePositionOffset=false
    Write Material Random=true
   )
   &AddVelocity(Velocity=<0 0 5>)
  ]
  particleUpdateScript [
   &ParticleState()
   &SkeletalMeshLocation(
    Bone / Socket Index=$Particles.PartitionParticles.Partition
    MaskSpawnGroup=false
    Mesh Sampling Type=::NewEnumerator5
    Position Sampling=::NewEnumerator1
    Sampled Position Offset=&SelectVectorFromArray(
     Direct Array Index=$Particles.PartitionParticles.Partition
     Vector Selection Array=DataInterface() {propertyValues {FloatData [<0 0 0>, <0 0 0>, <6.5 1 16>, <0 0 1>, <0 0 1>, <0 0 1>, <0 0 1>, <0 0 1>]}}
    )
    Skeletal Mesh=$System.Skeletal Mesh
    Socket Sampling Mode=::NewEnumerator5
   )
   &ScaleSpriteSize(Scale Sprite Size Mode=::NewEnumerator0, Uniform Scale Factor=2) {module=ScaleSpriteSize001}
   &SpriteRotationRate(Rotation Rate=&RandomRangeFloat(Minimum=-50, Maximum=50))
   &ScaleSpriteSize(
    Scale Sprite Size Mode=::NewEnumerator0
    Uniform Scale Factor=&FloatFromCurve(FloatCurve=DataInterface() {propertyValues {Curve {keys [<at 0 0.45>,<at 1 1.2>]}}}, Scale Curve=8)
   )
   &PointAttractionForce(
    Attraction Strength=0.25
    Attractor Position Offset:Vector3f=<0 0 75>
    Kill Within Radius=false
    Module.Use Falloff=true
    UseAttractorOffset=true
   )
   &SubUVAnimation(
    End Frame Range Override=30
    "Sprite Renderer "=DataInterface() {propertyValues {SpriteRenderer=&self:Smoke_Detail_1.NiagaraSpriteRendererProperties_0}}
    Start Frame Range Override=&RandomRangeInt(Minimum=0, Maximum=5)
    UseEndFrame=true
    UseStartFrame=true
   )
   &ScaleColor(
    Scale Alpha=&Multiply_Float(
     A=&RandomRangeFloat(Minimum=0.5, Maximum=1)
     B=&FloatFromCurve(FloatCurve=DataInterface() {propertyValues {Curve {keys [
          {arriveTangent=9.0725, interpMode=RCIM_Cubic, leaveTangent=9.0725, tangentMode=RCTM_User, time=0, value=0}
          {arriveTangent=-2.5249, interpMode=RCIM_Cubic, leaveTangent=-2.5249, tangentMode=RCTM_User, time=0.1408, value=0.9847}
          {arriveTangent=-0.0405, interpMode=RCIM_Cubic, leaveTangent=-0.0405, tangentMode=RCTM_User, time=1, value=0}
         ]}}})
    )
    ScaleA=true
    ScaleRGB=true
   )
   &GravityForce(Gravity=<0 0 35>) {module=GravityForce001}
   &Drag(Use Linear Drag=true, Use Rotational Drag=true)
   &ScaleSpriteSize(
    Scale Sprite Size Mode=::NewEnumerator0
    Uniform Scale Factor=&MakeFloatFromVector(Vector=$Engine.Owner.Scale, Channel=::NewEnumerator0)
   ) {module=ScaleSpriteSize002}
   &SolveForcesAndVelocity(Clamp Velocity=false, Limit Acceleration=false)
  ]
  rendererClasses [&NiagaraSpriteRendererProperties]
  renderers [{rendererClass=&NiagaraSpriteRendererProperties}]
 }
}
summary {
 systemName=NS_DrumKit_Alchemy
 emitters [
  {
   emitterName=FlameRibbon
   rendererClasses [
    &NiagaraRibbonRendererProperties
    &NiagaraSpriteRendererProperties
    &NiagaraMeshRendererProperties
   ]
  }
  {emitterName=FlameGlow, rendererClasses [&NiagaraSpriteRendererProperties, &NiagaraLightRendererProperties]}
  {emitterName=MagicStars, rendererClasses [&NiagaraSpriteRendererProperties, &NiagaraLightRendererProperties]}
  {emitterName=Alchemy_DrumKit_Potion_Liquid001, rendererClasses [&NiagaraMeshRendererProperties]}
  {emitterName=Smoke_Detail, rendererClasses [&NiagaraSpriteRendererProperties]}
  {
   emitterName=MagicGlowAndBigLight
   rendererClasses [&NiagaraSpriteRendererProperties, &NiagaraLightRendererProperties]
  }
 ]
 userVariables [
  User.bIsFrontEndPreview {type=NiagaraBool}
  User.Duration {type=NiagaraFloat, defaultValue=0.5, description=""}
  User.IsFilling {type=NiagaraBool, defaultValue=true}
  User.IsFinished {type=NiagaraBool}
 ]
}
